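# Niagara effect for the TigerRootHype backpack cosmetic that draws music-reactive, multicolored mesh blocks and a light.
# Two block emitters spawn short-lived meshes when gated by System Music V, spreading them along rows offset from the pack bone.
# Event audio amplitude from the Fortnite NPC drives block height and brightness, plus the color-cycling light's spawn and intensity.
# Time-of-day and front-end modifiers adjust brightness, and one block emitter adds separate glow color and size attributes.
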NiagaraSystem /BRCosmetics/Accessories/FORT_Backpacks/Backpack_M_MED_TigerRootHype/FX/NS_Backpack_TigerRootHype
dependencies {
 usedDataInterfaces [&NiagaraDataInterfaceSkeletalMesh, &NiagaraDataInterfaceColorCurve, &NiagaraDataInterfaceCurve]
 usedDynamicInputs []
 usedModules [
  &SystemState
  &self:SystemSpawnScript.SystemScriptSource.SystemScriptGraph.NiagaraNodeAssignment_0.SetVariables_#1
  &self:SystemSpawnScript.SystemScriptSource.SystemScriptGraph.NiagaraNodeAssignment_1.SetVariables_#2
 ]
}
emitters {
 Blocks_left_Ray {
  emitterUpdateScript [
   &EmitterState(
    Inactive Response=::NewEnumerator1
    Life Cycle Mode=::NewEnumerator1
    Loop Behavior=::NewEnumerator1
    Loop Duration Mode=::NewEnumerator1
    UseLoopDelay=false
   )
   &SpawnRate(
    SpawnRate=&MakeCustomFloatFromBool(Bool=&SetBoolByFloatComparison(A=$System.Music V, B=0.1), True Float=40, False Float=0)
    Use Spawn Probability=false
   )
  ]
  particleSpawnScript [
   &FN_SetSpawnGroup(Number of Spawn Groups=3)
   &InitializeParticle(
    Color=<rgba 2 2 2 2>
    Lifetime Max=0.4
    Lifetime Min=0.4
    Lifetime Mode=::NewEnumerator1
    Mesh Scale=&Multiply_VectorByFloat(
     Vector=&RandomRangeVector(Minimum=<0.4 0.4 0.8>, Maximum=<0.5 0.5 1>)
     Float=&RandomRangeFloat(Minimum=1, Maximum=1)
    )
    Mesh Scale Mode=::NewEnumerator4
    Sprite Size Mode=::NewEnumerator3
    UsePositionOffset=false
    Write Material Random=true
   )
   &SphereLocation(
    MaskSpawnGroup=false
    Override Seed=false
    Sphere Origin:NiagaraPosition=&ConvertVectorToPosition(Input Position=<0 0 0>, Vector Is In Absolute Large World Space:NiagaraBool=false)
    Sphere Radius=1
    Surface Only=false
    UseNonUniformScale=false
    UseOffset=false
    UseSphereOrientation=false
   )
   /SetVariables(Particles.Random=&RandomRangeFloat(Minimum=0, Maximum=1)) {isSetParameters=true, module=SetVariables_#3}
   &CameraOffset(Camera Offset Amount=-2)
  ]
  particleUpdateScript [
   &ParticleState()
   &SolveForcesAndVelocity(Clamp Velocity=false, Limit Acceleration=false)
   &SkeletalMeshLocation(
    Bone Sampling Mode=::NewEnumerator4
    MaskSpawnGroup=false
    Sampled Position Offset=&Add_Vector(
     A=&ConvertPositionToVector(Input Position=$Particles.Initial.Position, Convert To Absolute Large World Vector=false)
     B=&Add_Vector(
      A=&LerpPosition(A=<9 -1 -9>, B=<-9 -1 -9>, Alpha=$Particles.Random)
      B=&RandomRangeVector(Minimum=<0 -3 0>, Maximum=<0 0 0>)
     )
    )
    Skeletal Mesh=DataInterface() {propertyValues {
      DefaultMesh=null
      ExcludeBoneName=root
      OptionalComponentClass=null
      PreviewMesh=/BRCosmetics/Accessories/FORT_Backpacks/Backpack_M_MED_TigerRootHype/Meshes/M_MED_TigerRootHype_Pack
      SkinningMode=SkinOnTheFly
      SoftSourceActor=null
      SourceMode=Default
      UvSetIndex=0
      WholeMeshLOD=-1
      bExcludeBone=false
      bReadDeformedGeometry=true
      bRequireCurrentFrameData=true
      ComponentTags []
      FilteredBones [pack]
      FilteredSockets []
      MeshUserParameter {parameter {name=None, typeDefHandle {registeredTypeIndex=1}, varData []}}
      SamplingRegions []
     }}
   )
   &ScaleColor(
    Scale Alpha=&FloatFromCurve(
     FloatCurve=DataInterface() {propertyValues {Curve {keys [
         {interpMode=RCIM_Cubic, tangentMode=RCTM_User, time=0, value=0}
         {interpMode=RCIM_Cubic, tangentMode=RCTM_User, time=0.5, value=2}
         {tangentMode=RCTM_User, time=1, value=0}
        ]}}}
     CurveIndex=$Particles.Random
     Scale Curve=&Multiply_Float(A=&FN_FrontEndFloatModifier(Gameplay=3, Front End=3), B=1)
    )
    Scale RGB=&Multiply_VectorByFloat(
     Vector=&Multiply_VectorByFloat(
      Vector=&MakeVectorFromLinearColorRGB(Color=&ColorFromCurve(
        ColorCurve=DataInterface() {propertyValues {
          AlphaCurve {keys [<at 0 1 RCIM_Cubic>]}
          BlueCurve {keys [<at 0 2>,<at 0.1861 0>,<at 0.3508 0.0959>,<at 0.548 0.629>,<at 0.814 1>,<at 1 0 RCIM_Cubic>]}
          GreenCurve {keys [<at 0 0.0417>,<at 0.1861 0.8873>,<at 0.3508 0>,<at 0.548 1>,<at 0.814 0.3512>,<at 1 0.395 RCIM_Cubic>]}
          RedCurve {keys [<at 0 1.0208>,<at 0.1861 1>,<at 0.3508 1>,<at 0.548 0>,<at 0.814 0>,<at 1 1 RCIM_Cubic>]}
         }}
        CurveIndex=&Multiply_Float(A=$Particles.Random, B=$Particles.NormalizedAge)
       ))
      Float=&Lerp_Float(A=0.5, B=2, Alpha=&ClampFloat(Float=$NPC.FortniteNPC.EventAudio_Amplitude))
     )
     Float=&Multiply_Float(A=$System.TOD, B=&FN_FrontEndFloatModifier(Gameplay=0.5, Front End=0.5))
    )
    ScaleA=true
    ScaleRGB=true
   )
   &ScaleMeshSize(Scale Factor=&MakeVector(
     X=1
     Y=0.1
     Z=&FloatFromCurve(
      FloatCurve=DataInterface() {propertyValues {Curve {keys [
          {interpMode=RCIM_Cubic, tangentMode=RCTM_User, time=0, value=0}
          {interpMode=RCIM_Cubic, tangentMode=RCTM_User, time=0.5, value=1}
          {tangentMode=RCTM_User, time=1, value=0}
         ]}}}
      Scale Curve=&Lerp_Float(
       A=1
       B=&FN_FrontEndFloatModifier(Gameplay=2.5, Front End=2.5)
       Alpha=&ClampFloat(Float=$NPC.FortniteNPC.EventAudio_Amplitude)
      )
     )
    ))
   /SetVariables(Particles.Glow Color=&MakeLinearColorFromVectorAndFloat(
     "Vector (RGB)"=&Multiply_VectorByFloat(
      Vector=&Multiply_VectorByFloat(
       Vector=&MakeVectorFromLinearColorRGB(Color=&ColorFromCurve(
         ColorCurve=DataInterface() {propertyValues {
           AlphaCurve {keys [<at 0 1 RCIM_Cubic>]}
           BlueCurve {keys [<at 0 1 RCIM_Cubic>,<at 0.125 0.423>,<at 0.3868 1>,<at 0.635 0.7428>,<at 0.8208 1>,<at 1 1>]}
           GreenCurve {keys [<at 0 0.3612 RCIM_Cubic>,<at 0.125 0>,<at 0.3868 0.1916>,<at 0.635 0>,<at 0.8208 0.0238>,<at 1 0.3583>]}
           RedCurve {keys [<at 0 0 RCIM_Cubic>,<at 0.125 1>,<at 0.3868 0>,<at 0.635 1>,<at 0.8208 0>,<at 1 0>]}
          }}
         CurveIndex=&Multiply_Float(A=$Particles.Random, B=$Particles.NormalizedAge)
        ))
       Float=&Lerp_Float(A=1, B=1.2, Alpha=&ClampFloat(Float=$NPC.FortniteNPC.EventAudio_Amplitude))
      )
      Float=&Multiply_Float(A=$System.TOD, B=&FN_FrontEndFloatModifier(Gameplay=0.8, Front End=0.5))
     )
     "Float (Alpha)"=&Multiply_Float(
      A=&FloatFromCurve(
       FloatCurve=DataInterface() {propertyValues {Curve {keys [
           {interpMode=RCIM_Cubic, tangentMode=RCTM_User, time=0, value=0}
           <at 0.2 1>
           {interpMode=RCIM_Cubic, tangentMode=RCTM_User, time=0.8, value=1}
           {tangentMode=RCTM_User, time=1, value=0}
          ]}}}
       Scale Curve=&Multiply_Float(
        A=&FN_FrontEndFloatModifier(Gameplay=2, Front End=2)
        B=&RandomRangeFloat(Minimum=1, Maximum=1)
       )
      )
      B=20
     )
    )) {isSetParameters=true}
   /SetVariables(Particles.Glow Size:Vector3f=&Multiply_Vector(
     A=&Multiply_VectorByFloat(
      Vector=&RandomRangeVector(Minimum=<0.4 0.4 0.7>, Maximum=<0.4 0.4 1>)
      Float=&RandomRangeFloat(Minimum=1, Maximum=1)
     )
     B=&MakeVector(
      X=1
      Y=0.3
      Z=&FloatFromCurve(
       FloatCurve=DataInterface() {propertyValues {Curve {keys [
           {interpMode=RCIM_Cubic, tangentMode=RCTM_User, time=0, value=0.3}
           {interpMode=RCIM_Cubic, tangentMode=RCTM_User, time=0.5, value=1}
           {tangentMode=RCTM_User, time=1, value=0.3}
          ]}}}
       Scale Curve=&Lerp_Float(
        A=1.2
        B=&FN_FrontEndFloatModifier(Gameplay=3.2, Front End=3.2)
        Alpha=&ClampFloat(Float=$NPC.FortniteNPC.EventAudio_Amplitude)
       )
      )
     )
    )) {isSetParameters=true}
  ]
  rendererClasses [&NiagaraMeshRendererProperties]
  renderers [
   {rendererClass=&NiagaraMeshRendererProperties, rendererIndex=0}
   {rendererClass=&NiagaraMeshRendererProperties, rendererIndex=1}
  ]
 }
 Blocks_left_Ray001 {
  emitterUpdateScript [
   &EmitterState()
   &SpawnRate(
    SpawnRate=&MakeCustomFloatFromBool(Bool=&SetBoolByFloatComparison(A=$System.Music V, B=0.1), True Float=60, False Float=0)
    Use Spawn Probability=false
   )
  ]
  particleSpawnScript [
   &FN_SetSpawnGroup(Number of Spawn Groups=3)
   &InitializeParticle(
    Color=<rgba 2 2 2 5>
    Lifetime Max=0.4
    Lifetime Min=0.4
    Lifetime Mode=::NewEnumerator1
    Mesh Scale=&Multiply_VectorByFloat(
     Vector=&RandomRangeVector(Minimum=<0.2 2 1.5>, Maximum=<0.4 0.4 2>)
     Float=&RandomRangeFloat(Minimum=1, Maximum=1)
    )
    Mesh Scale Mode=::NewEnumerator4
    Sprite Size Mode=::NewEnumerator3
    UsePositionOffset=false
    Write Material Random=true
   )
   &SphereLocation(
    MaskSpawnGroup=false
    Override Seed=false
    Sphere Origin:NiagaraPosition=&ConvertVectorToPosition(Input Position=<0 0 0>, Vector Is In Absolute Large World Space:NiagaraBool=false)
    Sphere Radius=1
    Surface Only=false
    UseNonUniformScale=false
    UseOffset=false
    UseSphereOrientation=false
   )
   /SetVariables(Particles.Random=&RandomRangeFloat(Minimum=0, Maximum=1)) {isSetParameters=true, module=SetVariables_#3}
   &CameraOffset(Camera Offset Amount=-2)
  ]
  particleUpdateScript [
   &ParticleState()
   &SolveForcesAndVelocity(Clamp Velocity=false, Limit Acceleration=false)
   &SkeletalMeshLocation(
    Bone Sampling Mode=::NewEnumerator4
    MaskSpawnGroup=false
    Sampled Position Offset=&Add_Vector(
     A=&ConvertPositionToVector(Input Position=$Particles.Initial.Position, Convert To Absolute Large World Vector=false)
     B=&Add_Vector(
      A=&LerpPosition(A=<9 -1 0>, B=<-9 -1 0>, Alpha=$Particles.Random)
      B=&RandomRangeVector(Minimum=<0 -2 0>, Maximum=<0 0 0>)
     )
    )
    Skeletal Mesh=DataInterface() {propertyValues {
      DefaultMesh=null
      ExcludeBoneName=root
      OptionalComponentClass=null
      PreviewMesh=&M_MED_TigerRootHype_Pack
      SkinningMode=SkinOnTheFly
      SoftSourceActor=null
      SourceMode=Default
      UvSetIndex=0
      WholeMeshLOD=-1
      bExcludeBone=false
      bReadDeformedGeometry=true
      bRequireCurrentFrameData=true
      ComponentTags []
      FilteredBones [pack]
      FilteredSockets []
      MeshUserParameter {parameter {name=None, typeDefHandle {registeredTypeIndex=1}, varData []}}
      SamplingRegions []
     }}
   )
   &ScaleColor(
    Scale Alpha=&FloatFromCurve(
     FloatCurve=DataInterface() {propertyValues {Curve {keys [
         {interpMode=RCIM_Cubic, tangentMode=RCTM_User, time=0, value=0}
         {interpMode=RCIM_Cubic, tangentMode=RCTM_User, time=0.5, value=2}
         {tangentMode=RCTM_User, time=1, value=0}
        ]}}}
     CurveIndex=$Particles.Random
     Scale Curve=&Multiply_Float(A=&FN_FrontEndFloatModifier(Gameplay=1, Front End=2), B=1)
    )
    Scale RGB=&Multiply_VectorByFloat(
     Vector=&Multiply_VectorByFloat(
      Vector=&MakeVectorFromLinearColorRGB(Color=&ColorFromCurve(
        ColorCurve=DataInterface() {propertyValues {
          AlphaCurve {keys [<at 0 1 RCIM_Cubic>]}
          BlueCurve {keys [<at 0 2>,<at 0.1861 0>,<at 0.3508 0.0959>,<at 0.548 0.629>,<at 0.814 1>,<at 1 0 RCIM_Cubic>]}
          GreenCurve {keys [<at 0 0.0417>,<at 0.1861 0.8873>,<at 0.3508 0>,<at 0.548 1>,<at 0.814 0.3512>,<at 1 0.395 RCIM_Cubic>]}
          RedCurve {keys [<at 0 1.0208>,<at 0.1861 1>,<at 0.3508 1>,<at 0.548 0>,<at 0.814 0>,<at 1 1 RCIM_Cubic>]}
         }}
        CurveIndex=&Multiply_Float(A=$Particles.Random, B=$Particles.NormalizedAge)
       ))
      Float=&Lerp_Float(A=0.5, B=5, Alpha=&ClampFloat(Float=$NPC.FortniteNPC.EventAudio_Amplitude))
     )
     Float=&Multiply_Float(A=$System.TOD, B=&FN_FrontEndFloatModifier(Gameplay=1, Front End=0.5))
    )
    ScaleA=true
    ScaleRGB=true
   )
   &ScaleMeshSize(Scale Factor=&MakeVector(
     X=1
     Y=0.1
     Z=&FloatFromCurve(
      FloatCurve=DataInterface() {propertyValues {Curve {keys [
          {interpMode=RCIM_Cubic, tangentMode=RCTM_User, time=0, value=0}
          {interpMode=RCIM_Cubic, tangentMode=RCTM_User, time=0.5, value=1}
          {tangentMode=RCTM_User, time=1, value=0}
         ]}}}
      Scale Curve=&Lerp_Float(
       A=2
       B=&FN_FrontEndFloatModifier(Gameplay=2.5, Front End=2.5)
       Alpha=&ClampFloat(Float=$NPC.FortniteNPC.EventAudio_Amplitude)
      )
     )
    ))
  ]
  rendererClasses [&NiagaraMeshRendererProperties]
  renderers [{rendererClass=&NiagaraMeshRendererProperties, rendererIndex=0}]
 }
 Light {
  emitterUpdateScript [
   &EmitterState(Inactive Response=::NewEnumerator1, Life Cycle Mode=::NewEnumerator1, Loop Duration=1, UseLoopDelay=false)
   &SpawnBurst_Instantaneous(
    Spawn Count=&MakeIntFromFloat(Float=&MakeCustomFloatFromBool(Bool=&SetBoolByFloatComparison(A=$NPC.FortniteNPC.EventAudio_Amplitude, B=0), True Float=1, False Float=0))
    Use Spawn Probability=false
    UseLoopCountLimit=false
   )
  ]
  particleSpawnScript [&InitializeParticle(Sprite Size Mode=::NewEnumerator3, UsePositionOffset=false, Write Material Random=true)]
  particleUpdateScript [
   &ParticleState()
   &FN_Light_Attributes(
    Color=&ColorFromCurve(ColorCurve=DataInterface() {propertyValues {
       AlphaCurve {keys [<at 0 1 RCIM_Cubic>]}
       BlueCurve {keys [<at 0 1>,<at 0.2484 1>,<at 0.3668 0>,<at 0.53 1>,<at 0.6878 0.5>,<at 0.9085 1 RCIM_Cubic>]}
       GreenCurve {keys [<at 0 0>,<at 0.2484 0>,<at 0.3668 0.5>,<at 0.53 0>,<at 0.6878 1>,<at 0.9085 0 RCIM_Cubic>]}
       RedCurve {keys [<at 0 1>,<at 0.2484 0.1667>,<at 0.3668 1>,<at 0.53 0.1667>,<at 0.6878 0>,<at 0.9085 1 RCIM_Cubic>]}
      }})
    Create Color Attribute=true
    Create Exponent Attribute=true
    Create Position Attribute=true
    Create Radius Attribute=true
    Exponent=4.4
    Position=&AddVectorToPosition(Position=$Particles.Position, Vector=<0 -10 0>)
    Radius=40
    Scale Color=&Multiply_Float(
     A=&Multiply_Float(A=10, B=&FNTimeOfDayScalar(Night Scale=0.7, Morning Scale=1, Evening Scale=1))
     B=&Lerp_Float(A=0, B=1, Alpha=$NPC.FortniteNPC.EventAudio_Amplitude)
    )
   )
  ]
  rendererClasses [&NiagaraSpriteRendererProperties, &NiagaraLightRendererProperties]
  renderers [
   {rendererClass=&NiagaraSpriteRendererProperties}
   {rendererClass=&NiagaraLightRendererProperties, rendererIndex=1}
  ]
 }
}
summary {
 systemName=NS_Backpack_TigerRootHype
 emitters [
  {emitterName=Blocks_left_Ray, rendererClasses [&NiagaraMeshRendererProperties]}
  {emitterName=Light, rendererClasses [&NiagaraSpriteRendererProperties, &NiagaraLightRendererProperties]}
  {emitterName=Blocks_left_Ray001, rendererClasses [&NiagaraMeshRendererProperties]}
 ]
 userVariables [
  User.AuidoPreview {type=NiagaraFloat, defaultValue=0, description=""}
  User.bIsFrontEnd {type=NiagaraBool}
  User.bIsFrontEndPreview {type=NiagaraBool}
 ]
}
