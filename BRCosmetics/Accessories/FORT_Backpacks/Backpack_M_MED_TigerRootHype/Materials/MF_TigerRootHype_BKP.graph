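# Material function for the TigerRootHype backpack that layers an animated, audio-reactive swirl effect onto incoming material attributes.
# Screen-aligned, object-centered UVs and radial coordinates drive rotating, noise-warped color-curve gradients, scrolling stripe layers and a tree glow.
# Event and emote audio levels modulate bounce and color shift.
# Masked results set fresnel-shaded, time-of-day-scaled emissive and adjust specular and roughness, behind the Use_TigerRootHype static switch.
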MaterialFunction /BRCosmetics/Accessories/FORT_Backpacks/Backpack_M_MED_TigerRootHype/Materials/MF_TigerRootHype_BKP
graph {nodes [
  Constant2Vector() {r=-0.5}
  n1 = LinearInterpolate(
   A=ScalarParameter() {group=Wrap_TigerRoot_Audio, parameterName=TigR_Audio_Tree_Bounce_A}
   B=ScalarParameter() {defaultValue=0.5, group=Wrap_TigerRoot_Audio, parameterName=TigR_Audio_Tree_Bounce_B}
   Alpha=Multiply(A=ScalarParameter() {defaultValue=0.25, group=Wrap_TigerRoot_Audio, parameterName=TigR_Audio_Tree_Bounce_Strenght}, B=NamedRerouteUsage())
  )
  n2 = NamedRerouteUsage()
  n3 = VectorParameter() {parameterName="TigR_UVs offset   A-Scale", defaultValue=<rgba -0.5 -0.5 0 1>}
  n4 = Multiply(A=Add(A=ComponentMask(None=@n2) {b=true, r=true}, B=ComponentMask(None=@n3.RGB) {g=true, r=true}), B=@n3.A)
  n5 = Frac(None=Add(A=MaterialFunctionCall(Vector or UVs=@n4) {materialFunction=&VectorToRadialValue}.Vector Converted to Angle) {constB=0.75})
  n6 = Multiply(
   A=Abs(None=Add(A=@n5) {constB=-0.5})
   B=ScalarParameter() {defaultValue=3, group=Wrap_TigerRoot_BackGround, parameterName=TigR_Tree_TileRate}
  )
  n7 = Add(A=@n1, B=@n6)
  n8 = Multiply(A=Time(), B=ScalarParameter() {defaultValue=-0.1, group=Wrap_TigerRoot_BackGround, parameterName=TigR_Tree_Rotation_Speed})
  n9 = Frac(None=ComponentMask(None=Add(A=@n7, B=@n8)) {r=true})
  CurveAtlasRowParameter(CurveTime=@n9) {
   atlas=&AthenaCurveAtlas01
   curve=/Game/Weapons/WeaponSkins/Wraps/Textures/0_ColorCurves/CC_TigR_ColorsA
   group=Wrap_TigerRoot_BackGround
   parameterName=TigR_TreeA_CC
  }
  n10 = ScalarParameter() {defaultValue=-0.1, group=Wrap_TigerRoot_BackGround, parameterName=TigR_Audio_ColorShift_Strength}
  n11 = NamedRerouteUsage()
  n12 = MaterialFunctionCall(Vector or UVs=NamedRerouteUsage()) {materialFunction=&VectorToRadialValue}
  n13 = Power(
   Base=ComponentMask(None=@n12.Radial Coordinates) {g=true}
   Exp=ScalarParameter() {defaultValue=0.2, group=Wrap_TigerRoot_BackGround, parameterName=TigR_Rotation_Pow}
  )
  n14 = ScalarParameter() {defaultValue=5, group=Wrap_TigerRoot_BackGround, parameterName=TigR_Rotation_Amount}
  n15 = Time()
  n16 = ScalarParameter() {defaultValue=-0.25, group=Wrap_TigerRoot_BackGround, parameterName=TigR_Rotation_Speed}
  n17 = Add(A=@n12.Radial Coordinates, B=Add(A=Multiply(A=@n13, B=@n14), B=Multiply(A=@n15, B=@n16)))
  n18 = ComponentMask(None=@n17) {r=true}
  n19 = ComponentMask(None=VectorParameter() {group=Wrap_TigerRoot_BackGround, parameterName=TigR_Bck_TileRate, defaultValue=<rgba 1 0.2 0 1>}.RGB) {g=true, r=true}
  n20 = ComponentMask(None=VectorParameter() {group=Wrap_TigerRoot_BackGround, parameterName=TigR_Bck_Speed, defaultValue=<rgba 0 0.1 0 1>}.RGB) {g=true, r=true}
  n21 = Time()
  n22 = NamedRerouteUsage()
  n23 = TextureSample!(
   UVs=Add(A=Multiply(A=@n17, B=@n19), B=Multiply(A=@n20, B=@n21))
   Tex=NamedRerouteUsage()
   "DDX(UVs)"=DDX(Value=@n22)
   "DDY(UVs)"=DDY(Value=@n22)
   Apply View MipBias
  ) {
   mipValueMode=TMVM_Derivative
   samplerType=SAMPLERTYPE_Masks
   texture=/BRCosmetics/Weapons/WeaponSkins/Wraps/Textures/TigerRoot/T_Wrap_TigerRoot_Msk
  }
  n24 = Multiply(A=Add(A=@n23.A) {constB=-0.5}, B=ScalarParameter() {defaultValue=1, group=Wrap_TigerRoot_BackGround, parameterName=TigR_Bck_Warp_Amount})
  n25 = Frac(None=Add(A=Multiply(A=@n10, B=@n11), B=Add(A=@n18, B=@n24)))
  CurveAtlasRowParameter(CurveTime=@n25) {
   atlas=&AthenaCurveAtlas01
   curve=/Game/Characters/Player/Female/Medium/Bodies/F_MED_Loofah/Textures/T_Loofah_Colors_02
   defaultValue=211
   group=Wrap_TigerRoot_BackGround
   parameterName=TigR_BckA_CC
  }
  n26 = NamedRerouteUsage()
  n27 = Multiply(A=ComponentMask(None=NamedRerouteUsage()) {r=true}, B=NamedRerouteUsage())
  n28 = Multiply(A=ComponentMask(None=NamedRerouteUsage()) {g=true}, B=NamedRerouteUsage())
  n29 = Fresnel(ExponentIn=ScalarParameter() {defaultValue=5, group=Wrap_TigerRoot, parameterName=TigR_Fresnel_Exp}) {baseReflectFraction=0}
  n30 = Add(A=Power(Base=OneMinus(None=@n29)), B=ScalarParameter() {defaultValue=0.2, group=Wrap_TigerRoot, parameterName=TigR_Fresnel_DarkValues}) {constB=0.5}
  n31 = Constant() {r=0.5}
  n32 = Constant() {r=1}
  n33 = Constant() {r=0.5}
  n34 = Constant() {r=0.5}
  n35 = Multiply(
   A=Multiply(A=Add(A=@n27, B=@n28), B=Saturate(None=@n30))
   B=Multiply(A=MaterialFunctionCall(Value_Morning=@n31, Value_Day=@n32, Value_Evening=@n33, Value_Night=@n34) {materialFunction=&Scalar_Time_of_Day}.Result)
  )
  n36 = NamedRerouteUsage()
  n37 = LinearInterpolate(
   A=@n36
   B=Multiply(A=@n36, B=ScalarParameter() {defaultValue=0.1, group=Wrap_TigerRoot, parameterName=TigR_Spec_Adjustment})
   Alpha=ComponentMask(None=NamedRerouteUsage()) {r=true}
  )
  n38 = NamedRerouteUsage()
  n39 = LinearInterpolate(
   A=@n38
   B=Multiply(A=@n38, B=ScalarParameter() {defaultValue=0.75, group=Wrap_TigerRoot, parameterName=TigR_Roughness_Adjustment})
   Alpha=ComponentMask(None=NamedRerouteUsage()) {r=true}
  )
  FunctionOutput(None=StaticSwitchParameter(
    True=SetMaterialAttributes!(MaterialAttributes=@n26, EmissiveColor=@n35, Specular=@n37, Roughness=@n39) {attributeSetTypes [#1, #2, #3]}
    False=NamedRerouteUsage()
   ) {group=Wrap_TigerRoot, parameterName=Use_TigerRootHype})
  n40 = CollectionParameter() {collection=&FortniteMaterialParameters, parameterName=EventAudio_Amplitude}
  n41 = CollectionParameter() {collection=&FortniteMaterialParameters, parameterName=Emote_AudioLow}
  n42 = Max(A=@n40, B=@n41)
  n43 = ScalarParameter() {defaultValue=0.25, group=Wrap_TigerRoot_Audio, parameterName=TigR_Sectrum_Low_Remap, sortPriority=150}
  n44 = ScalarParameter() {defaultValue=1, group=Wrap_TigerRoot_Audio, parameterName=TigR_Sectrum_High_Remap, sortPriority=150}
  n45 = Constant()
  n46 = Constant() {r=1}
  n47 = CollectionParameter() {collection=&FortniteMaterialParameters, parameterName=Emote_AudioHigh}
  n48 = Max(
   A=Saturate(None=MaterialFunctionCall(Input=@n42, Input Low=@n43, Input High=@n44, Target Low=@n45, Target High=@n46) {materialFunction=&RemapValueRange}.Result)
   B=@n47
  )
  n49 = Time()
  n50 = Add(
   A=ScalarParameter() {defaultValue=-9999999, group=Wrap_Reactive, parameterName=TimeFrom-BP----------------, sortPriority=600}
   B=ScalarParameter() {group=Wrap_TigerRoot_Reactive, parameterName=TigR_reactive_Delay}
  )
  n51 = Saturate(None=Divide(A=Subtract(A=@n49, B=@n50), B=ScalarParameter() {defaultValue=0.5, group=Wrap_TigerRoot_Reactive, parameterName=TigR_reactive_Duration}))
  n52 = ScalarParameter() {defaultValue=1, group=Wrap_TigerRoot_Reactive, parameterName=TigR_reactive_Falloff}
  Max(
   A=@n48
   B=LinearInterpolate(
    A=ScalarParameter() {defaultValue=1, group=Wrap_TigerRoot_Reactive, parameterName=TigR_reactive_Start}
    B=ScalarParameter() {group=Wrap_TigerRoot_Reactive, parameterName=TigR_reactive_End}
    Alpha=Power(Base=@n51, Exp=@n52)
   )
  )
  Multiply(A=@n1, B=@n6)
  Multiply(A=TextureSample!(UVs=@n4, Tex, Apply View MipBias) {texture=/Game/Weapons/WeaponSkins/Wraps/Textures/UVCheaker}.RGB)
  n53 = MaterialFunctionCall() {materialFunction=/Game/Packages/Fortress_SharedMaterials/Base_Material_Functions/BaseCharacterShaderFunctions/Audio/MF_AudioReactivity}
  Multiply(A=NamedRerouteDeclaration(None=@n53.Amplitude) {name=Audio_Amplitude, nodeColor=<rgba 0 0.0879 1 1>}, B=@n53.Spectrum)
  NamedRerouteDeclaration(None=FunctionInput(Preview=WorldPosition().XYZ) {bUsePreviewValueAsDefault=true, inputName=PreSkin_Pos}) {name=PreSkin Pos, nodeColor=<rgba 0 0.4609 1 1>}
  NamedRerouteDeclaration(None=TextureSample(Tex=TextureObjectParameter() {group=Skin FX, parameterName=SkinFX_Mask, texture=&T_BlacksRGB}) {samplerType=SAMPLERTYPE_Masks}.RGBA) {name=SkinFX Msk, nodeColor=<rgba 1 0.1641 0 1>}
  NamedRerouteDeclaration(None=@n17) {name=Uvs_Rotation, nodeColor=<rgba 1 0.3281 0 1>}
  n54 = VectorParameter() {group=Wrap_TigerRoot_Stripes, parameterName=TigR_StripsC_TileRate, defaultValue=<rgba 1 0.005 0 1>}
  n55 = NamedRerouteUsage()
  n56 = TextureSample!(
   UVs=Multiply(A=NamedRerouteUsage(), B=ComponentMask(None=@n54.RGB) {g=true, r=true})
   Tex=NamedRerouteUsage()
   "DDX(UVs)"=DDX(Value=@n55)
   "DDY(UVs)"=DDY(Value=@n55)
   Apply View MipBias
  ) {
   mipValueMode=TMVM_Derivative
   samplerType=SAMPLERTYPE_Masks
   texture=&T_Wrap_TigerRoot_Msk
  }
  NamedRerouteDeclaration(None=@n56.G) {name=Strips_C, nodeColor=<rgba 0 1 0.6484 1>}
  n57 = CurveAtlasRowParameter(CurveTime=@n25) {
   atlas=/Game/Weapons/WeaponSkins/Wraps/Textures/CruveAtlas_Wraps_01
   curve=&CC_TigR_ColorsA
   defaultValue=129
   group=Wrap_TigerRoot_BackGround
   parameterName=TigR_Bck_CC
  }
  n58 = LinearInterpolate(
   A=MaterialFunctionCall!(
    Index=ScalarParameter() {defaultValue=129, desc=look under 00_ColorCurve, group=Wrap_TigerRoot_BackGround, parameterName=TigR_Bck_CC_Index}
    CCinput=@n57
    CCinput_Alpha
    UV=@n25
   ) {materialFunction=/Game/Weapons/WeaponSkins/Wraps/MaterialFunctions/MF_Global_CC_Tex}.3V
   B=VectorParameter() {group=Wrap_TigerRoot_Stripes, parameterName=TigR_StripsB_Color, defaultValue=<rgba 1.25 0 0.9817 1>}.RGB
   Alpha=NamedRerouteUsage()
  )
  n59 = Multiply(
   A=VectorParameter() {group=Wrap_TigerRoot_Stripes, parameterName=TigR_StripsA_Color, defaultValue=<rgba 0.5 0.5 0.5 1>}.RGB
   B=NamedRerouteUsage()
  )
  n60 = NamedRerouteUsage()
  n61 = NamedRerouteUsage()
  NamedRerouteDeclaration(None=LinearInterpolate(
    A=Add(A=@n58, B=@n59)
    B=VectorParameter() {group=Wrap_TigerRoot_Stripes, parameterName=TigR_StripsC_Color, defaultValue=<rgba 10 10 10 0.5>}.RGB
    Alpha=Multiply(A=@n60, B=@n61) {constB=0}
   )) {name=Swirls, nodeColor=<rgba 1 0 0.75 1>}
  n62 = ScalarParameter() {defaultValue=0.5, group=Wrap_TigerRoot_Audio, parameterName=TigR_Audio_Bounce_Strenght}
  n63 = NamedRerouteUsage()
  n64 = Divide(B=MaterialFunctionCall() {materialFunction=&ObjectScale}.Scale Z)
  n65 = NamedRerouteUsage()
  n66 = MaterialFunctionCall!(Center Point World Position=TransformPosition(None=ComponentMask(None=@n65) {b=true, g=true, r=true}) {transformType=TRANSFORMPOSSOURCE_World}) {materialFunction=/Engine/Functions/Engine_MaterialFunctions02/UVs/WorldSpaceAlignedScreenCoordinates}
  n67 = Constant() {r=-2500}
  n68 = PreSkinnedLocalBounds()
  n69 = MaterialFunctionCall(Float3=Subtract(A=@n68.Max, B=@n68.Max)) {materialFunction=&BreakOutFloat3Components}
  n70 = Multiply(A=ComponentMask(None=@n65) {a=true}, B=Max(A=Max(A=@n69.R, B=@n69.R), B=@n69.B))
  n71 = Multiply(
   A=LinearInterpolate(
    A=ScalarParameter() {defaultValue=1, group=Wrap_TigerRoot_Audio, parameterName=TigR_Audio_Bounce_A}
    B=ScalarParameter() {defaultValue=1.5, group=Wrap_TigerRoot_Audio, parameterName=TigR_Audio_Bounce_B}
    Alpha=Multiply(A=@n62, B=@n63)
   )
   B=Multiply(A=Multiply(A=@n64, B=@n66.Offset and Scaled UVS), B=Divide(A=@n67, B=@n70))
  ) {constA=1}
  NamedRerouteDeclaration(None=@n71) {name=UVs_VP_Warped, nodeColor=<rgba 0 1 0.5078 1>}
  n72 = VertexInterpolator(VS=PreSkinnedPosition())
  n73 = PreSkinnedLocalBounds()
  n74 = MaterialFunctionCall(Color=Subtract(A=@n73.Max, B=@n73.Max)) {materialFunction=&SplitComponents}
  n75 = Max(A=@n74.R, B=@n74.R)
  NamedRerouteDeclaration(None=Divide(A=Subtract(A=@n72.PS, B=@n73.Min), B=Max(A=@n75, B=@n74.B))) {name=Cal_Bounds, nodeColor=<rgba 0.7109 0 1 1>}
  n76 = VectorParameter() {group=Wrap_TigerRoot, parameterName=TigR_Position_Scale, defaultValue=<rgba 0 0 1 -1>}
  NamedRerouteDeclaration(None=AppendVector(A=@n76.RGB, B=@n76.RGB)) {name=SafeZone VP, nodeColor=<rgba 0.5 1 0 1>}
  NamedRerouteDeclaration() {name=SafeZone Character-Switch, nodeColor=<rgba 1 0.9375 0 1>}
  NamedRerouteDeclaration(None=@n48) {name=AudioSpectrum, nodeColor=<rgba 1 0.8203 0 1>}
  n77 = Add(A=NamedRerouteUsage(), B=Constant2Vector() {g=0.23, r=0.3}) {constB=0.3}
  n78 = ComponentMask(None=VectorParameter() {group=Wrap_TigerRoot_Stripes, parameterName=TigR_StripsB_TileRate, defaultValue=<rgba 1 0.005 0 1>}.RGB) {g=true, r=true}
  n79 = ComponentMask(None=VectorParameter() {group=Wrap_TigerRoot_Stripes, parameterName=TigR_StripsB_Speed, defaultValue=<rgba -0.135 0.03 0 1>}.RGB) {g=true, r=true}
  n80 = Time()
  n81 = NamedRerouteUsage()
  n82 = TextureSample(
   UVs=Add(A=Multiply(A=@n77, B=@n78), B=Multiply(A=@n79, B=@n80))
   Tex=NamedRerouteUsage()
   "DDX(UVs)"=DDX(Value=@n81)
   "DDY(UVs)"=DDY(Value=@n81)
  ) {
   mipValueMode=TMVM_Derivative
   samplerType=SAMPLERTYPE_Masks
   texture=&T_Wrap_TigerRoot_Msk
  }
  NamedRerouteDeclaration(None=@n82.R) {name=Streams_B, nodeColor=<rgba 1 0 0.0937 1>}
  n83 = NamedRerouteUsage()
  n84 = ComponentMask(None=VectorParameter() {group=Wrap_TigerRoot_Stripes, parameterName=TigR_StripsA_TileRate, defaultValue=<rgba 1 0.0051 0 1>}.RGB) {g=true, r=true}
  n85 = ComponentMask(None=VectorParameter() {group=Wrap_TigerRoot_Stripes, parameterName=TigR_StripsA_Speed, defaultValue=<rgba -0.25 0.05 0 1>}.RGB) {g=true, r=true}
  n86 = Time()
  n87 = NamedRerouteUsage()
  n88 = TextureSample(
   UVs=Add(A=Multiply(A=@n83, B=@n84), B=Multiply(A=@n85, B=@n86))
   Tex=NamedRerouteUsage()
   "DDX(UVs)"=DDX(Value=@n87)
   "DDY(UVs)"=DDY(Value=@n87)
  ) {
   mipValueMode=TMVM_Derivative
   samplerType=SAMPLERTYPE_Masks
   texture=&T_Wrap_TigerRoot_Msk
  }
  NamedRerouteDeclaration(None=@n88.R) {name=Streams_A, nodeColor=<rgba 1 0 0.0938 1>}
  n89 = ScalarParameter() {defaultValue=131, desc=look under 00_ColorCurve, group=Wrap_TigerRoot_BackGround, parameterName=TigR_Tree_CC_Index}
  n90 = CurveAtlasRowParameter(CurveTime=@n9) {
   atlas=&CruveAtlas_Wraps_01
   curve=/Game/Weapons/WeaponSkins/Wraps/Textures/0_ColorCurves/CC_TigR_ColorsB
   defaultValue=131
   group=Wrap_TigerRoot_BackGround
   parameterName=TigR_Tree_CC
  }
  NamedRerouteDeclaration(None=Multiply(
    A=MaterialFunctionCall(Index=@n89, CCinput=@n90, UV=@n9) {materialFunction=&MF_Global_CC_Tex}.3V
    B=ScalarParameter() {defaultValue=1.5, group=Wrap_TigerRoot_BackGround, parameterName=TigR_Tree_Brightness}
   )) {name=outer Glow, nodeColor=<rgba 0 1 0.7891 1>}
  n91 = GetMaterialAttributes(None=FunctionInput() {inputType=FunctionInput_MaterialAttributes}) {attributeGetTypes [#1, #2, #3]}
  NamedRerouteDeclaration(None=@n91.MaterialAttributes) {name=Root_Material, nodeColor=<rgba 0.5234 0 1 1>}
  NamedRerouteDeclaration(None=@n91.EmissiveColor) {name=Root_Emissive, nodeColor=<rgba 0.4297 1 0 1>}
  NamedRerouteDeclaration(None=@n91.Specular) {name=Root_Spec, nodeColor=<rgba 1 0.0234 0 1>}
  NamedRerouteDeclaration(None=TextureObjectParameter() {group=Wrap_TigerRoot, parameterName=TigR__Tex_MskA, texture=&T_BlacksRGB}) {name=Tex_MskA, nodeColor=<rgba 1 0.8672 0 1>}
  NamedRerouteDeclaration(None=@n91.Roughness) {name=Root_Roughness, nodeColor=<rgba 1 0.0234 0 1>}
  NamedRerouteDeclaration(None=@n71) {name=UVs_VP, nodeColor=<rgba 0 1 0.5078 1>}
  NamedRerouteUsage()
  ScalarParameter() {defaultValue=-9999999, group=TopSideProjection_Emissive_Reactive, parameterName=Time_BP}
 ]}
